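# Physics asset for the FineCheek two-handed pickaxe, giving its hanging feather ornaments capsule bodies for simulated secondary motion.
# A knot body joins the handle and acts as the shared root for eight feather chains, each starting at a master bone.
# Six thin feathers continue through four segments and two other feathers through two.
# Constraints link each body to its parent with tight twist and moderate swing limits, slightly looser along the two-segment feathers.
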PhysicsAsset /BRCosmetics/Weapons/FORT_Melee/Pickaxe_FineCheek/Meshes/FineCheek_Axe_Physics
bodies {
 Melee_TwoHanded_Handle [{shapeType=Capsule, length=1, radius=1, shapeName=Melee_TwoHanded_Handle_capsule}]
 dyn_1_feather_thin_01 [{
   shapeType=Capsule
   length=2.1964
   radius=0.896
   shapeName=dyn_1_feather_thin_01_capsule
   center=<2.0623 0.0934 -0.1288>
   rotation=<rot 0 -89.2164 -89.5886>
  }]
 dyn_1_feather_thin_02 [{
   shapeType=Capsule
   length=1.5751
   radius=0.7826
   shapeName=dyn_1_feather_thin_02_capsule
   center=<2.2218 -0.2058 -0.039>
   rotation=<rot 0 -90.5172 -87.7707>
  }]
 dyn_1_feather_thin_03 [{
   shapeType=Capsule
   length=1.5751
   radius=0.7826
   shapeName=dyn_1_feather_thin_03_capsule
   center=<2.0567 -0.1233 0.1016>
   rotation=<rot 0 84.0111 -86.7125>
  }]
 dyn_1_feather_thin_04 [{
   shapeType=Capsule
   length=1.5751
   radius=0.7826
   shapeName=dyn_1_feather_thin_04_capsule
   center=<1.2481 0.0759 -0.0718>
   rotation=<rot 0 -78.4862 -83.7401>
  }]
 dyn_1_feather_thin_MSTR [{shapeType=Capsule, length=1.4582, radius=0.9303, shapeName=dyn_1_feather_thin_MSTR_capsule}]
 dyn_2_feather_thin_01 [{
   shapeType=Capsule
   length=4.1445
   radius=0.896
   shapeName=dyn_2_feather_thin_01_capsule
   center=<3.6427 0.2695 0.0396>
   rotation=<rot 0 -95.618 -87.9847>
  }]
 dyn_2_feather_thin_02 [{
   shapeType=Capsule
   length=1.5751
   radius=0.7826
   shapeName=dyn_2_feather_thin_02_capsule
   center=<3.102 -0.0441 0.0479>
   rotation=<rot 0 -83.2751 -87.2235>
  }]
 dyn_2_feather_thin_03 [{
   shapeType=Capsule
   length=2.0347
   radius=0.7826
   shapeName=dyn_2_feather_thin_03_capsule
   center=<1.614 0.0025 -0.2305>
   rotation=<rot 0 -91.2788 -86.4932>
  }]
 dyn_2_feather_thin_04 [{
   shapeType=Capsule
   length=1.5751
   radius=0.7826
   shapeName=dyn_2_feather_thin_04_capsule
   center=<-0.2266 0.0886 -0.0823>
   rotation=<rot 0 94.3036 -78.2031>
  }]
 dyn_2_feather_thin_MSTR [{shapeType=Capsule, length=1.1109, radius=0.7826, shapeName=dyn_2_feather_thin_MSTR_capsule, rotation=<rot 0 -90 0>}]
 dyn_3_feather_thin_01 [{
   shapeType=Capsule
   length=2.1964
   radius=0.896
   shapeName=dyn_3_feather_thin_01_capsule
   center=<1.9495 -0.031 -0.059>
   rotation=<rot 0 -89.4558 -85.9056>
  }]
 dyn_3_feather_thin_02 [{
   shapeType=Capsule
   length=1.5751
   radius=0.7826
   shapeName=dyn_3_feather_thin_02_capsule
   center=<2.8081 -0.1203 0.1865>
   rotation=<rot 0 86.3369 -86.627>
  }]
 dyn_3_feather_thin_03 [{
   shapeType=Capsule
   length=1.5751
   radius=0.7826
   shapeName=dyn_3_feather_thin_03_capsule
   center=<2.3095 -0.121 -0.0207>
   rotation=<rot 0 82.6509 -86.8425>
  }]
 dyn_3_feather_thin_04 [{
   shapeType=Capsule
   length=1.5751
   radius=0.7826
   shapeName=dyn_3_feather_thin_04_capsule
   center=<1.5998 -0.3925 -0.3209>
   rotation=<rot 0 -89.2389 -72.0765>
  }]
 dyn_3_feather_thin_MSTR [{shapeType=Capsule, length=1.1109, radius=0.7826, shapeName=dyn_3_feather_thin_MSTR_capsule, rotation=<rot 0 -90 0>}]
 dyn_4_feather_thin_01 [{
   shapeType=Capsule
   length=2.1964
   radius=0.896
   shapeName=dyn_4_feather_thin_01_capsule
   center=<3.0322 -0.0341 0.0351>
   rotation=<rot 0 -81.9384 -89.2891>
  }]
 dyn_4_feather_thin_02 [{
   shapeType=Capsule
   length=2.6541
   radius=0.7826
   shapeName=dyn_4_feather_thin_02_capsule
   center=<2.8996 -0.0179 -0.1303>
   rotation=<rot 0 92.4134 -86.0866>
  }]
 dyn_4_feather_thin_03 [{
   shapeType=Capsule
   length=1.5751
   radius=0.7826
   shapeName=dyn_4_feather_thin_03_capsule
   center=<1.4537 0.0363 -0.0297>
   rotation=<rot 0 -95.6724 -89.5619>
  }]
 dyn_4_feather_thin_04 [{
   shapeType=Capsule
   length=1.5751
   radius=0.7826
   shapeName=dyn_4_feather_thin_04_capsule
   center=<1.296 -0.0154 -0.1681>
   rotation=<rot 0 -83.9038 -82.2821>
  }]
 dyn_4_feather_thin_MSTR [{shapeType=Capsule, length=1.1109, radius=0.7826, shapeName=dyn_4_feather_thin_MSTR_capsule, rotation=<rot 0 -90 0>}]
 dyn_5_feather_thin_01 [{
   shapeType=Capsule
   length=2.4141
   radius=0.896
   shapeName=dyn_5_feather_thin_01_capsule
   center=<2.7597 -0.0971 -0.0186>
   rotation=<rot 0 95.1764 -79.7563>
  }]
 dyn_5_feather_thin_02 [{
   shapeType=Capsule
   length=2.9445
   radius=0.7826
   shapeName=dyn_5_feather_thin_02_capsule
   center=<3.3054 -0.1444 0.043>
   rotation=<rot 0 89.7538 -87.9778>
  }]
 dyn_5_feather_thin_03 [{
   shapeType=Capsule
   length=1.5751
   radius=0.7826
   shapeName=dyn_5_feather_thin_03_capsule
   center=<2.1757 0.118 0.0981>
   rotation=<rot 0 81.419 -88.8526>
  }]
 dyn_5_feather_thin_04 [{
   shapeType=Capsule
   length=1.5751
   radius=0.7826
   shapeName=dyn_5_feather_thin_04_capsule
   center=<0.7399 -0.4198 -0.2151>
   rotation=<rot 0 -104.291 -75.2953>
  }]
 dyn_5_feather_thin_MSTR [{shapeType=Capsule, length=1.1109, radius=0.7826, shapeName=dyn_5_feather_thin_MSTR_capsule, rotation=<rot 0 -90 0>}]
 dyn_6_feather_thin_01 [{
   shapeType=Capsule
   length=2.1964
   radius=0.896
   shapeName=dyn_6_feather_thin_01_capsule
   center=<2.0696 -0.0453 -0.0428>
   rotation=<rot 0 -91.226 -87.6692>
  }]
 dyn_6_feather_thin_02 [{
   shapeType=Capsule
   length=1.5751
   radius=0.7826
   shapeName=dyn_6_feather_thin_02_capsule
   center=<1.191 -0.0186 0.0677>
   rotation=<rot 0 92.1411 -86.5354>
  }]
 dyn_6_feather_thin_03 [{
   shapeType=Capsule
   length=1.5751
   radius=0.7826
   shapeName=dyn_6_feather_thin_03_capsule
   center=<1.065 0.197 0.0096>
   rotation=<rot 0 -78.4664 -85.6057>
  }]
 dyn_6_feather_thin_04 [{
   shapeType=Capsule
   length=1.5751
   radius=0.7826
   shapeName=dyn_6_feather_thin_04_capsule
   center=<1.1552 -0.2224 0.298>
   rotation=<rot 0 80.0639 -74.7667>
  }]
 dyn_6_feather_thin_MSTR [{shapeType=Capsule, length=1.1109, radius=0.7826, shapeName=dyn_6_feather_thin_MSTR_capsule, rotation=<rot 0 -90 0>}]
 dyn_feather1_01 [{shapeType=Capsule, length=2.2778, radius=0.5, shapeName=dyn_feather1_01_capsule, center=<3.1696 0 0>, rotation=<rot 0 90 90>}]
 dyn_feather1_02 [{
   shapeType=Capsule
   length=3.3319
   radius=0.7826
   shapeName=dyn_feather1_02_capsule
   center=<1.5233 -0.3091 0.3169>
   rotation=<rot 0 75.6411 -71.9036>
  }]
 dyn_feather1_MSTR [{shapeType=Capsule, length=1.1109, radius=0.7826, shapeName=dyn_feather1_MSTR_capsule, rotation=<rot 0 -90 0>}]
 dyn_feather2_01 [{
   shapeType=Capsule
   length=2.3832
   radius=0.5
   shapeName=dyn_feather2_01_capsule
   center=<3.7554 0.0728 -0.1979>
   rotation=<rot 0 132.2538 -89.7454>
  }]
 dyn_feather2_02 [{
   shapeType=Capsule
   length=2.2967
   radius=0.7826
   shapeName=dyn_feather2_02_capsule
   center=<2.1355 0.3346 0.3753>
   rotation=<rot 0 100.6519 -75.6344>
  }]
 dyn_feather2_MSTR [{shapeType=Capsule, length=1.1109, radius=0.7826, shapeName=dyn_feather2_MSTR_capsule, rotation=<rot 0 -90 0>}]
 dyn_knot_MSTR [{shapeType=Capsule, length=1, radius=1, shapeName=dyn_knot_MSTR_capsule}]
}
constraints [
 {bone1Name=dyn_knot_MSTR, bone2Name=Melee_TwoHanded_Handle, swing1LimitDegrees=45, swing2LimitDegrees=45, twistLimitDegrees=5}
 {bone1Name=dyn_1_feather_thin_MSTR, bone2Name=dyn_knot_MSTR, swing1LimitDegrees=45, swing2LimitDegrees=45, twistLimitDegrees=5}
 {bone1Name=dyn_1_feather_thin_01, bone2Name=dyn_1_feather_thin_MSTR, swing1LimitDegrees=45, swing2LimitDegrees=45, twistLimitDegrees=5}
 {bone1Name=dyn_1_feather_thin_02, bone2Name=dyn_1_feather_thin_01, swing1LimitDegrees=45, swing2LimitDegrees=45, twistLimitDegrees=5}
 {bone1Name=dyn_1_feather_thin_03, bone2Name=dyn_1_feather_thin_02, swing1LimitDegrees=45, swing2LimitDegrees=45, twistLimitDegrees=5}
 {bone1Name=dyn_1_feather_thin_04, bone2Name=dyn_1_feather_thin_03, swing1LimitDegrees=45, swing2LimitDegrees=45, twistLimitDegrees=5}
 {bone1Name=dyn_2_feather_thin_MSTR, bone2Name=dyn_knot_MSTR, swing1LimitDegrees=45, swing2LimitDegrees=45, twistLimitDegrees=5}
 {bone1Name=dyn_2_feather_thin_01, bone2Name=dyn_2_feather_thin_MSTR, swing1LimitDegrees=45, swing2LimitDegrees=45, twistLimitDegrees=5}
 {bone1Name=dyn_2_feather_thin_02, bone2Name=dyn_2_feather_thin_01, swing1LimitDegrees=45, swing2LimitDegrees=45, twistLimitDegrees=5}
 {bone1Name=dyn_2_feather_thin_03, bone2Name=dyn_2_feather_thin_02, swing1LimitDegrees=45, swing2LimitDegrees=45, twistLimitDegrees=5}
 {bone1Name=dyn_2_feather_thin_04, bone2Name=dyn_2_feather_thin_03, swing1LimitDegrees=45, swing2LimitDegrees=45, twistLimitDegrees=5}
 {bone1Name=dyn_3_feather_thin_MSTR, bone2Name=dyn_knot_MSTR, swing1LimitDegrees=45, swing2LimitDegrees=45, twistLimitDegrees=5}
 {bone1Name=dyn_3_feather_thin_01, bone2Name=dyn_3_feather_thin_MSTR, swing1LimitDegrees=45, swing2LimitDegrees=45, twistLimitDegrees=5}
 {bone1Name=dyn_3_feather_thin_02, bone2Name=dyn_3_feather_thin_01, swing1LimitDegrees=45, swing2LimitDegrees=45, twistLimitDegrees=5}
 {bone1Name=dyn_3_feather_thin_03, bone2Name=dyn_3_feather_thin_02, swing1LimitDegrees=45, swing2LimitDegrees=45, twistLimitDegrees=5}
 {bone1Name=dyn_3_feather_thin_04, bone2Name=dyn_3_feather_thin_03, swing1LimitDegrees=45, swing2LimitDegrees=45, twistLimitDegrees=5}
 {bone1Name=dyn_4_feather_thin_MSTR, bone2Name=dyn_knot_MSTR, swing1LimitDegrees=45, swing2LimitDegrees=45, twistLimitDegrees=5}
 {bone1Name=dyn_4_feather_thin_01, bone2Name=dyn_4_feather_thin_MSTR, swing1LimitDegrees=45, swing2LimitDegrees=45, twistLimitDegrees=5}
 {bone1Name=dyn_4_feather_thin_02, bone2Name=dyn_4_feather_thin_01, swing1LimitDegrees=45, swing2LimitDegrees=45, twistLimitDegrees=5}
 {bone1Name=dyn_4_feather_thin_03, bone2Name=dyn_4_feather_thin_02, swing1LimitDegrees=45, swing2LimitDegrees=45, twistLimitDegrees=5}
 {bone1Name=dyn_4_feather_thin_04, bone2Name=dyn_4_feather_thin_03, swing1LimitDegrees=45, swing2LimitDegrees=45, twistLimitDegrees=5}
 {bone1Name=dyn_5_feather_thin_MSTR, bone2Name=dyn_knot_MSTR, swing1LimitDegrees=45, swing2LimitDegrees=45, twistLimitDegrees=5}
 {bone1Name=dyn_5_feather_thin_01, bone2Name=dyn_5_feather_thin_MSTR, swing1LimitDegrees=45, swing2LimitDegrees=45, twistLimitDegrees=5}
 {bone1Name=dyn_5_feather_thin_02, bone2Name=dyn_5_feather_thin_01, swing1LimitDegrees=45, swing2LimitDegrees=45, twistLimitDegrees=5}
 {bone1Name=dyn_5_feather_thin_03, bone2Name=dyn_5_feather_thin_02, swing1LimitDegrees=45, swing2LimitDegrees=45, twistLimitDegrees=5}
 {bone1Name=dyn_5_feather_thin_04, bone2Name=dyn_5_feather_thin_03, swing1LimitDegrees=45, swing2LimitDegrees=45, twistLimitDegrees=5}
 {bone1Name=dyn_6_feather_thin_MSTR, bone2Name=dyn_knot_MSTR, swing1LimitDegrees=45, swing2LimitDegrees=45, twistLimitDegrees=5}
 {bone1Name=dyn_6_feather_thin_01, bone2Name=dyn_6_feather_thin_MSTR, swing1LimitDegrees=45, swing2LimitDegrees=45, twistLimitDegrees=5}
 {bone1Name=dyn_6_feather_thin_02, bone2Name=dyn_6_feather_thin_01, swing1LimitDegrees=45, swing2LimitDegrees=45, twistLimitDegrees=5}
 {bone1Name=dyn_6_feather_thin_03, bone2Name=dyn_6_feather_thin_02, swing1LimitDegrees=45, swing2LimitDegrees=45, twistLimitDegrees=5}
 {bone1Name=dyn_6_feather_thin_04, bone2Name=dyn_6_feather_thin_03, swing1LimitDegrees=45, swing2LimitDegrees=45, twistLimitDegrees=5}
 {bone1Name=dyn_feather1_MSTR, bone2Name=dyn_knot_MSTR, swing1LimitDegrees=45, swing2LimitDegrees=45, twistLimitDegrees=5}
 {bone1Name=dyn_feather1_01, bone2Name=dyn_feather1_MSTR, swing1LimitDegrees=50, swing2LimitDegrees=50, twistLimitDegrees=10}
 {bone1Name=dyn_feather1_02, bone2Name=dyn_feather1_01, swing1LimitDegrees=50, swing2LimitDegrees=50, twistLimitDegrees=10}
 {bone1Name=dyn_feather2_MSTR, bone2Name=dyn_knot_MSTR, swing1LimitDegrees=45, swing2LimitDegrees=45, twistLimitDegrees=5}
 {bone1Name=dyn_feather2_01, bone2Name=dyn_feather2_MSTR, swing1LimitDegrees=50, swing2LimitDegrees=50, twistLimitDegrees=10}
 {bone1Name=dyn_feather2_02, bone2Name=dyn_feather2_01, swing1LimitDegrees=50, swing2LimitDegrees=50, twistLimitDegrees=10}
]
