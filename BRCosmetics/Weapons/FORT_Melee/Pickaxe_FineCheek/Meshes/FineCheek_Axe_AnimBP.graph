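# Anim Blueprint for the FineCheek pickaxe that adds secondary physics motion to its dangling knot and feather bones.
# The cached input pose feeds two solvers: a RigidBody node using the axe's physics asset in the front end, and an Anim Dynamics chain otherwise.
# The chain solves bone by bone with angular springs, gravity overrides and spherical limits between feathers.
# The EventGraph sets Jumping from jump or fall state, but the AnimGraph never reads it.
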AnimBlueprint /BRCosmetics/Weapons/FORT_Melee/Pickaxe_FineCheek/Meshes/FineCheek_Axe_AnimBP
target_skeleton=&/FineCheek_Axe_Skeleton
graphs [
 AnimGraph {nodes [
   n1 = Anim Dynamics - Bone: dyn_knot_MSTR(ComponentPose=Local To Component(LocalPose=Use cached pose 'Base'()), Alpha=1, ExternalForce="0,0,0") {
    angularDampingOverride=0.999
    angularSpringConstant=50
    bAngularSpring=true
    bOverrideAngularDamping=true
    bUseGravityOverride=true
    bUseSphericalLimits=true
    alphaBoolBlend {blendOption=Linear}
    boundBone {boneName=dyn_knot_MSTR}
    componentAppliedLinearAccClamp=<2000 2000 2000>
    componentLinearAccScale=<0.3 0.3 0.3>
    componentLinearVelScale=<0.3 0.3 0.3>
    gravityOverride {z=-500}
    physicsBodyDefinitions [{
      boundBone {boneName=dyn_knot_MSTR}
      boxExtents=<12 12 12>
      constraintSetup {angularLimitsMax=<10 30 30>, angularLimitsMin=<-10 -50 -50>, angularTarget=<0 0 0>}
      localJointOffset=<1 0 0>
     }]
    sphericalLimits [
     {limitRadius=1, drivingBone {boneName=dyn_feather2_01}, sphereLocalOffset=<0 0 0>}
     {limitRadius=1, drivingBone {boneName=dyn_feather2_02}, sphereLocalOffset=<0 0 0>}
     {limitRadius=1, drivingBone {boneName=dyn_5_feather_thin_01}, sphereLocalOffset=<0 0 0>}
     {limitRadius=1, drivingBone {boneName=dyn_6_feather_thin_01}, sphereLocalOffset=<0 0 0>}
     {limitRadius=1, drivingBone {boneName=dyn_4_feather_thin_02}, sphereLocalOffset=<0 0 0>}
    ]
   }
   n2 = "Transform (Modify) Bone - Bone: dyn_knot_MSTR"!(ComponentPose=@n1, bAlphaBoolEnabled:bool=True, Translation="0,0,0", Rotation="60,60,60", Scale="1,1,1") -> (Pose) {
    alphaInputType=Bool
    rotationMode=BMM_Additive
    rotationSpace=BCS_BoneSpace
    alphaBoolBlend {blendInTime=0.1, blendOption=Sinusoidal, blendOutTime=0.1}
    boneToModify {boneName=dyn_knot_MSTR}
    rotation=<rot 60 60 60>
   }
   n3 = Anim Dynamics - Bone: dyn_feather1_01(ComponentPose=@n2, Alpha=1, ExternalForce="0,0,0") {
    angularDampingOverride=0.999
    angularSpringConstant=50
    bAngularSpring=true
    bOverrideAngularDamping=true
    bUseGravityOverride=true
    bUseSphericalLimits=true
    alphaBoolBlend {blendOption=Linear}
    boundBone {boneName=dyn_feather1_01}
    componentAppliedLinearAccClamp=<2000 2000 2000>
    componentLinearAccScale=<0.2 0.2 0.2>
    componentLinearVelScale=<0.2 0.2 0.2>
    gravityOverride {z=-500}
    physicsBodyDefinitions [{
      boundBone {boneName=dyn_feather1_01}
      boxExtents=<12 12 12>
      constraintSetup {angularLimitsMax=<0 10 10>, angularLimitsMin=<-5 -10 -10>, angularTarget=<0 0 0>}
      localJointOffset=<1 0 0>
     }]
    sphericalLimits [
     {limitRadius=1, drivingBone {boneName=dyn_feather2_01}, sphereLocalOffset=<0 0 0>}
     {limitRadius=1, drivingBone {boneName=dyn_feather2_02}, sphereLocalOffset=<0 0 0>}
     {limitRadius=1, drivingBone {boneName=dyn_5_feather_thin_01}, sphereLocalOffset=<0 0 0>}
     {limitRadius=1, drivingBone {boneName=dyn_6_feather_thin_01}, sphereLocalOffset=<0 0 0>}
     {limitRadius=1, drivingBone {boneName=dyn_4_feather_thin_02}, sphereLocalOffset=<0 0 0>}
    ]
   }
   n4 = Anim Dynamics - Bone: dyn_feather1_02(ComponentPose=@n3, Alpha=1, ExternalForce="0,0,0") {
    angularDampingOverride=0.999
    angularSpringConstant=50
    bAngularSpring=true
    bOverrideAngularDamping=true
    bUseGravityOverride=true
    bUseSphericalLimits=true
    alphaBoolBlend {blendOption=Linear}
    boundBone {boneName=dyn_feather1_02}
    componentAppliedLinearAccClamp=<2000 2000 2000>
    componentLinearAccScale=<0.3 0.3 0.3>
    componentLinearVelScale=<0.3 0.3 0.3>
    gravityOverride {z=-500}
    physicsBodyDefinitions [{
      boundBone {boneName=dyn_feather1_02}
      boxExtents=<12 12 12>
      constraintSetup {angularLimitsMax=<5 10 10>, angularLimitsMin=<-15 -10 -10>, angularTarget=<0 0 0>}
      localJointOffset=<1 0 0>
     }]
    sphericalLimits [
     {limitRadius=1, drivingBone {boneName=dyn_feather2_01}, sphereLocalOffset=<0 0 0>}
     {limitRadius=1, drivingBone {boneName=dyn_feather2_02}, sphereLocalOffset=<0 0 0>}
     {limitRadius=1, drivingBone {boneName=dyn_5_feather_thin_01}, sphereLocalOffset=<0 0 0>}
     {limitRadius=1, drivingBone {boneName=dyn_6_feather_thin_01}, sphereLocalOffset=<0 0 0>}
     {limitRadius=1, drivingBone {boneName=dyn_4_feather_thin_02}, sphereLocalOffset=<0 0 0>}
    ]
   }
   n5 = Anim Dynamics - Bone: dyn_feather2_01(ComponentPose=@n4, Alpha=1, ExternalForce="0,0,0") {
    angularDampingOverride=0.999
    angularSpringConstant=200
    bAngularSpring=true
    bOverrideAngularDamping=true
    bUseGravityOverride=true
    bUsePlanarLimit=false
    bUseSphericalLimits=true
    alphaBoolBlend {blendOption=Linear}
    boundBone {boneName=dyn_feather2_01}
    componentAppliedLinearAccClamp=<2000 2000 2000>
    componentLinearAccScale=<0.2 0.2 0.2>
    componentLinearVelScale=<0.2 0.2 0.2>
    gravityOverride {z=-500}
    physicsBodyDefinitions [{
      boundBone {boneName=dyn_feather2_01}
      boxExtents=<12 12 12>
      constraintSetup {angularLimitsMax=<0 30 30>, angularLimitsMin=<-5 -30 -30>, angularTarget=<0 0 0>}
      localJointOffset=<1 0 0>
     }]
    sphericalLimits [
     {limitRadius=1, drivingBone {boneName=dyn_feather1_01}, sphereLocalOffset=<0 0 0>}
     {limitRadius=1, drivingBone {boneName=dyn_feather1_02}, sphereLocalOffset=<0 0 0>}
     {limitRadius=1, drivingBone {boneName=dyn_5_feather_thin_01}, sphereLocalOffset=<0 0 0>}
     {limitRadius=1, drivingBone {boneName=dyn_6_feather_thin_01}, sphereLocalOffset=<0 0 0>}
     {limitRadius=1, drivingBone {boneName=dyn_4_feather_thin_01}, sphereLocalOffset=<0 0 0>}
     {limitRadius=1, drivingBone {boneName=dyn_4_feather_thin_02}, sphereLocalOffset=<-0.8955 0.5366 0.28>}
    ]
   }
   n6 = Anim Dynamics - Bone: dyn_feather2_02(ComponentPose=@n5, Alpha=1, ExternalForce="0,0,0") {
    angularDampingOverride=0.999
    angularSpringConstant=200
    bAngularSpring=true
    bOverrideAngularDamping=true
    bUseGravityOverride=true
    bUsePlanarLimit=false
    bUseSphericalLimits=true
    alphaBoolBlend {blendOption=Linear}
    boundBone {boneName=dyn_feather2_02}
    componentAppliedLinearAccClamp=<2000 2000 2000>
    componentLinearAccScale=<0.3 0.3 0.3>
    componentLinearVelScale=<0.3 0.3 0.3>
    gravityOverride {z=-500}
    physicsBodyDefinitions [{
      boundBone {boneName=dyn_feather2_02}
      boxExtents=<12 12 12>
      constraintSetup {angularLimitsMax=<5 35 35>, angularLimitsMin=<-15 -35 -35>, angularTarget=<0 0 0>}
      localJointOffset=<1 0 0>
     }]
    sphericalLimits [
     {limitRadius=1, drivingBone {boneName=dyn_feather1_01}, sphereLocalOffset=<0 0 0>}
     {limitRadius=1, drivingBone {boneName=dyn_feather1_02}, sphereLocalOffset=<0 0 0>}
     {limitRadius=1, drivingBone {boneName=dyn_5_feather_thin_01}, sphereLocalOffset=<0 0 0>}
     {limitRadius=1, drivingBone {boneName=dyn_6_feather_thin_01}, sphereLocalOffset=<0 0 0>}
     {limitRadius=1, drivingBone {boneName=dyn_4_feather_thin_01}, sphereLocalOffset=<0 0 0>}
     {limitRadius=1, drivingBone {boneName=dyn_4_feather_thin_02}, sphereLocalOffset=<1 0 0>}
    ]
   }
   n7 = Anim Dynamics - Bone: dyn_2_feather_thin_01(ComponentPose=@n6, Alpha=1, ExternalForce="0,0,0") {
    angularDampingOverride=0.99
    angularSpringConstant=300
    bAngularSpring=true
    bOverrideAngularDamping=true
    bUseGravityOverride=true
    bUsePlanarLimit=false
    bUseSphericalLimits=true
    alphaBoolBlend {blendOption=Linear}
    boundBone {boneName=dyn_2_feather_thin_01}
    componentAppliedLinearAccClamp=<1000 1000 1000>
    componentLinearAccScale=<0.1 0.1 0.1>
    componentLinearVelScale=<0.1 0.1 0.1>
    gravityOverride {z=-700}
    physicsBodyDefinitions [{
      boundBone {boneName=dyn_2_feather_thin_01}
      boxExtents=<12 12 12>
      constraintSetup {angularLimitsMax=<0 20 20>, angularLimitsMin=<-5 -5 -5>, angularTarget=<0 0 0>}
      localJointOffset=<2 0 0>
     }]
    sphericalLimits [
     {limitRadius=1, drivingBone {boneName=dyn_feather1_01}, sphereLocalOffset=<3 0 0>}
     {limitRadius=1, drivingBone {boneName=dyn_feather1_02}, sphereLocalOffset=<3 0 0>}
     {limitRadius=1, drivingBone {boneName=dyn_feather2_01}, sphereLocalOffset=<3 0 0>}
     {limitRadius=1, drivingBone {boneName=dyn_feather2_02}, sphereLocalOffset=<3 0 0>}
    ]
   }
   n8 = Anim Dynamics - Bone: dyn_2_feather_thin_02(ComponentPose=@n7, Alpha=1, ExternalForce="0,0,0") {
    angularDampingOverride=0.99
    angularSpringConstant=300
    bAngularSpring=true
    bOverrideAngularDamping=true
    bUseGravityOverride=true
    bUsePlanarLimit=false
    bUseSphericalLimits=true
    alphaBoolBlend {blendOption=Linear}
    boundBone {boneName=dyn_2_feather_thin_02}
    componentAppliedLinearAccClamp=<1000 1000 1000>
    componentLinearAccScale=<0.1 0.1 0.1>
    componentLinearVelScale=<0.1 0.1 0.1>
    gravityOverride {z=-700}
    physicsBodyDefinitions [{
      boundBone {boneName=dyn_2_feather_thin_02}
      boxExtents=<12 12 12>
      constraintSetup {angularLimitsMax=<0 20 20>, angularLimitsMin=<-15 -15 15>, angularTarget=<0 0 0>}
      localJointOffset=<2 0 0>
     }]
    sphericalLimits [
     {limitRadius=1, drivingBone {boneName=dyn_feather1_01}, sphereLocalOffset=<3 0 0>}
     {limitRadius=1, drivingBone {boneName=dyn_feather1_02}, sphereLocalOffset=<3 0 0>}
     {limitRadius=1, drivingBone {boneName=dyn_feather2_01}, sphereLocalOffset=<3 0 0>}
     {limitRadius=1, drivingBone {boneName=dyn_feather2_02}, sphereLocalOffset=<3 0 0>}
    ]
   }
   n9 = Anim Dynamics - Bone: dyn_2_feather_thin_03(ComponentPose=@n8, Alpha=1, ExternalForce="0,0,0") {
    angularDampingOverride=0.99
    angularSpringConstant=300
    bAngularSpring=true
    bOverrideAngularDamping=true
    bUseGravityOverride=true
    bUsePlanarLimit=false
    bUseSphericalLimits=true
    alphaBoolBlend {blendOption=Linear}
    boundBone {boneName=dyn_2_feather_thin_03}
    componentAppliedLinearAccClamp=<1500 1500 1500>
    componentLinearAccScale=<0.25 0.25 0.25>
    componentLinearVelScale=<0.25 0.25 25>
    gravityOverride {z=-700}
    physicsBodyDefinitions [{
      boundBone {boneName=dyn_2_feather_thin_03}
      boxExtents=<12 12 12>
      constraintSetup {angularLimitsMax=<10 40 40>, angularLimitsMin=<-15 -40 -40>, angularTarget=<0 0 0>}
      localJointOffset=<2 0 0>
     }]
    sphericalLimits [
     {limitRadius=1, drivingBone {boneName=dyn_feather1_01}, sphereLocalOffset=<3 0 0>}
     {limitRadius=1, drivingBone {boneName=dyn_feather1_02}, sphereLocalOffset=<3 0 0>}
     {limitRadius=1, drivingBone {boneName=dyn_feather2_01}, sphereLocalOffset=<3 0 0>}
     {limitRadius=1, drivingBone {boneName=dyn_feather2_02}, sphereLocalOffset=<3 0 0>}
    ]
   }
   n10 = Anim Dynamics - Bone: dyn_2_feather_thin_04(ComponentPose=@n9, Alpha=1, ExternalForce="0,0,0") {
    angularDampingOverride=0.99
    angularSpringConstant=300
    bAngularSpring=true
    bOverrideAngularDamping=true
    bUseGravityOverride=true
    bUsePlanarLimit=false
    bUseSphericalLimits=true
    alphaBoolBlend {blendOption=Linear}
    boundBone {boneName=dyn_2_feather_thin_04}
    componentAppliedLinearAccClamp=<2000 2000 2000>
    componentLinearAccScale=<0.3 0.3 0.3>
    componentLinearVelScale=<0.3 0.3 0.3>
    gravityOverride {z=-700}
    physicsBodyDefinitions [{
      boundBone {boneName=dyn_2_feather_thin_04}
      boxExtents=<12 12 12>
      constraintSetup {angularLimitsMax=<10 40 40>, angularLimitsMin=<-15 -40 -40>, angularTarget=<0 0 0>}
      localJointOffset=<2 0 0>
     }]
    sphericalLimits [
     {limitRadius=1, drivingBone {boneName=dyn_feather1_01}, sphereLocalOffset=<3 0 0>}
     {limitRadius=1, drivingBone {boneName=dyn_feather1_02}, sphereLocalOffset=<3 0 0>}
     {limitRadius=1, drivingBone {boneName=dyn_feather2_01}, sphereLocalOffset=<3 0 0>}
     {limitRadius=1, drivingBone {boneName=dyn_feather2_02}, sphereLocalOffset=<3 0 0>}
    ]
   }
   n11 = Anim Dynamics - Bone: dyn_1_feather_thin_01(ComponentPose=@n10, Alpha=1, ExternalForce="0,0,0") {
    angularDampingOverride=0.99
    angularSpringConstant=300
    bAngularSpring=true
    bOverrideAngularDamping=true
    bUseGravityOverride=true
    bUsePlanarLimit=false
    bUseSphericalLimits=true
    alphaBoolBlend {blendOption=Linear}
    boundBone {boneName=dyn_1_feather_thin_01}
    componentAppliedLinearAccClamp=<1000 1000 1000>
    componentLinearAccScale=<0.1 0.1 0.1>
    componentLinearVelScale=<0.1 0.1 0.1>
    gravityOverride {z=-700}
    physicsBodyDefinitions [{
      boundBone {boneName=dyn_1_feather_thin_01}
      boxExtents=<12 12 12>
      constraintSetup {angularLimitsMax=<0 5 5>, angularLimitsMin=<-5 -5 -5>, angularTarget=<0 0 0>}
      localJointOffset=<2 0 0>
     }]
    sphericalLimits [
     {limitRadius=1, drivingBone {boneName=dyn_feather1_01}, sphereLocalOffset=<3 0 0>}
     {limitRadius=1, drivingBone {boneName=dyn_feather1_02}, sphereLocalOffset=<3 0 0>}
     {limitRadius=1, drivingBone {boneName=dyn_feather2_01}, sphereLocalOffset=<3 0 0>}
     {limitRadius=1, drivingBone {boneName=dyn_feather2_02}, sphereLocalOffset=<3 0 0>}
    ]
   }
   n12 = Anim Dynamics - Bone: dyn_1_feather_thin_02(ComponentPose=@n11, Alpha=1, ExternalForce="0,0,0") {
    angularDampingOverride=0.99
    angularSpringConstant=300
    bAngularSpring=true
    bOverrideAngularDamping=true
    bUseGravityOverride=true
    bUsePlanarLimit=false
    bUseSphericalLimits=true
    alphaBoolBlend {blendOption=Linear}
    boundBone {boneName=dyn_1_feather_thin_02}
    componentAppliedLinearAccClamp=<1000 1000 1000>
    componentLinearAccScale=<0.1 0.1 0.1>
    componentLinearVelScale=<0.1 0.1 0.1>
    gravityOverride {z=-700}
    physicsBodyDefinitions [{
      boundBone {boneName=dyn_1_feather_thin_02}
      boxExtents=<12 12 12>
      constraintSetup {angularLimitsMax=<0 20 20>, angularLimitsMin=<-15 -15 15>, angularTarget=<0 0 0>}
      localJointOffset=<2 0 0>
     }]
    sphericalLimits [
     {limitRadius=1, drivingBone {boneName=dyn_feather1_01}, sphereLocalOffset=<3 0 0>}
     {limitRadius=1, drivingBone {boneName=dyn_feather1_02}, sphereLocalOffset=<3 0 0>}
     {limitRadius=1, drivingBone {boneName=dyn_feather2_01}, sphereLocalOffset=<3 0 0>}
     {limitRadius=1, drivingBone {boneName=dyn_feather2_02}, sphereLocalOffset=<3 0 0>}
    ]
   }
   n13 = Anim Dynamics - Bone: dyn_1_feather_thin_03(ComponentPose=@n12, Alpha=1, ExternalForce="0,0,0") {
    angularDampingOverride=0.99
    angularSpringConstant=300
    bAngularSpring=true
    bOverrideAngularDamping=true
    bUseGravityOverride=true
    bUsePlanarLimit=false
    bUseSphericalLimits=true
    alphaBoolBlend {blendOption=Linear}
    boundBone {boneName=dyn_1_feather_thin_03}
    componentAppliedLinearAccClamp=<1500 1500 1500>
    componentLinearAccScale=<0.25 0.25 0.25>
    componentLinearVelScale=<0.25 0.25 25>
    gravityOverride {z=-700}
    physicsBodyDefinitions [{
      boundBone {boneName=dyn_1_feather_thin_03}
      boxExtents=<12 12 12>
      constraintSetup {angularLimitsMax=<10 40 40>, angularLimitsMin=<-15 -40 -40>, angularTarget=<0 0 0>}
      localJointOffset=<2 0 0>
     }]
    sphericalLimits [
     {limitRadius=1, drivingBone {boneName=dyn_feather1_01}, sphereLocalOffset=<3 0 0>}
     {limitRadius=1, drivingBone {boneName=dyn_feather1_02}, sphereLocalOffset=<3 0 0>}
     {limitRadius=1, drivingBone {boneName=dyn_feather2_01}, sphereLocalOffset=<3 0 0>}
     {limitRadius=1, drivingBone {boneName=dyn_feather2_02}, sphereLocalOffset=<3 0 0>}
    ]
   }
   n14 = Anim Dynamics - Bone: dyn_1_feather_thin_04(ComponentPose=@n13, Alpha=1, ExternalForce="0,0,0") {
    angularDampingOverride=0.99
    angularSpringConstant=300
    bAngularSpring=true
    bOverrideAngularDamping=true
    bUseGravityOverride=true
    bUsePlanarLimit=false
    bUseSphericalLimits=true
    alphaBoolBlend {blendOption=Linear}
    boundBone {boneName=dyn_1_feather_thin_04}
    componentAppliedLinearAccClamp=<2000 2000 2000>
    componentLinearAccScale=<0.3 0.3 0.3>
    componentLinearVelScale=<0.3 0.3 0.3>
    gravityOverride {z=-700}
    physicsBodyDefinitions [{
      boundBone {boneName=dyn_1_feather_thin_04}
      boxExtents=<12 12 12>
      constraintSetup {angularLimitsMax=<10 40 40>, angularLimitsMin=<-15 -40 -40>, angularTarget=<0 0 0>}
      localJointOffset=<2 0 0>
     }]
    sphericalLimits [
     {limitRadius=1, drivingBone {boneName=dyn_feather1_01}, sphereLocalOffset=<3 0 0>}
     {limitRadius=1, drivingBone {boneName=dyn_feather1_02}, sphereLocalOffset=<3 0 0>}
     {limitRadius=1, drivingBone {boneName=dyn_feather2_01}, sphereLocalOffset=<3 0 0>}
     {limitRadius=1, drivingBone {boneName=dyn_feather2_02}, sphereLocalOffset=<3 0 0>}
    ]
   }
   n15 = Anim Dynamics - Bone: dyn_3_feather_thin_01(ComponentPose=@n14, Alpha=1, ExternalForce="0,0,0") {
    angularDampingOverride=0.99
    angularSpringConstant=300
    bAngularSpring=true
    bOverrideAngularDamping=true
    bUseGravityOverride=true
    bUsePlanarLimit=false
    bUseSphericalLimits=true
    alphaBoolBlend {blendOption=Linear}
    boundBone {boneName=dyn_3_feather_thin_01}
    componentAppliedLinearAccClamp=<1000 1000 1000>
    componentLinearAccScale=<0.1 0.1 0.1>
    componentLinearVelScale=<0.1 0.1 0.1>
    gravityOverride {z=-700}
    physicsBodyDefinitions [{
      boundBone {boneName=dyn_3_feather_thin_01}
      boxExtents=<12 12 12>
      constraintSetup {angularLimitsMax=<0 5 5>, angularLimitsMin=<-5 -5 -5>, angularTarget=<0 0 0>}
      localJointOffset=<2 0 0>
     }]
    sphericalLimits [
     {limitRadius=1, drivingBone {boneName=dyn_feather1_01}, sphereLocalOffset=<3 0 0>}
     {limitRadius=1, drivingBone {boneName=dyn_feather1_02}, sphereLocalOffset=<3 0 0>}
     {limitRadius=1, drivingBone {boneName=dyn_feather2_01}, sphereLocalOffset=<3 0 0>}
     {limitRadius=1, drivingBone {boneName=dyn_feather2_02}, sphereLocalOffset=<3 0 0>}
    ]
   }
   n16 = Anim Dynamics - Bone: dyn_3_feather_thin_02(ComponentPose=@n15, Alpha=1, ExternalForce="0,0,0") {
    angularDampingOverride=0.99
    angularSpringConstant=300
    bAngularSpring=true
    bOverrideAngularDamping=true
    bUseGravityOverride=true
    bUsePlanarLimit=false
    bUseSphericalLimits=true
    alphaBoolBlend {blendOption=Linear}
    boundBone {boneName=dyn_3_feather_thin_02}
    componentAppliedLinearAccClamp=<1000 1000 1000>
    componentLinearAccScale=<0.1 0.1 0.1>
    componentLinearVelScale=<0.1 0.1 0.1>
    gravityOverride {z=-700}
    physicsBodyDefinitions [{
      boundBone {boneName=dyn_3_feather_thin_02}
      boxExtents=<12 12 12>
      constraintSetup {angularLimitsMax=<0 20 20>, angularLimitsMin=<-15 -15 15>, angularTarget=<0 0 0>}
      localJointOffset=<2 0 0>
     }]
    sphericalLimits [
     {limitRadius=1, drivingBone {boneName=dyn_feather1_01}, sphereLocalOffset=<3 0 0>}
     {limitRadius=1, drivingBone {boneName=dyn_feather1_02}, sphereLocalOffset=<3 0 0>}
     {limitRadius=1, drivingBone {boneName=dyn_feather2_01}, sphereLocalOffset=<3 0 0>}
     {limitRadius=1, drivingBone {boneName=dyn_feather2_02}, sphereLocalOffset=<3 0 0>}
    ]
   }
   n17 = Anim Dynamics - Bone: dyn_3_feather_thin_03(ComponentPose=@n16, Alpha=1, ExternalForce="0,0,0") {
    angularDampingOverride=0.99
    angularSpringConstant=300
    bAngularSpring=true
    bOverrideAngularDamping=true
    bUseGravityOverride=true
    bUsePlanarLimit=false
    bUseSphericalLimits=true
    alphaBoolBlend {blendOption=Linear}
    boundBone {boneName=dyn_3_feather_thin_03}
    componentAppliedLinearAccClamp=<1500 1500 1500>
    componentLinearAccScale=<0.25 0.25 0.25>
    componentLinearVelScale=<0.25 0.25 25>
    gravityOverride {z=-700}
    physicsBodyDefinitions [{
      boundBone {boneName=dyn_3_feather_thin_03}
      boxExtents=<12 12 12>
      constraintSetup {angularLimitsMax=<10 40 40>, angularLimitsMin=<-15 -40 -40>, angularTarget=<0 0 0>}
      localJointOffset=<2 0 0>
     }]
    sphericalLimits [
     {limitRadius=1, drivingBone {boneName=dyn_feather1_01}, sphereLocalOffset=<3 0 0>}
     {limitRadius=1, drivingBone {boneName=dyn_feather1_02}, sphereLocalOffset=<3 0 0>}
     {limitRadius=1, drivingBone {boneName=dyn_feather2_01}, sphereLocalOffset=<3 0 0>}
     {limitRadius=1, drivingBone {boneName=dyn_feather2_02}, sphereLocalOffset=<3 0 0>}
    ]
   }
   n18 = Anim Dynamics - Bone: dyn_3_feather_thin_04(ComponentPose=@n17, Alpha=1, ExternalForce="0,0,0") {
    angularDampingOverride=0.99
    angularSpringConstant=300
    bAngularSpring=true
    bOverrideAngularDamping=true
    bUseGravityOverride=true
    bUsePlanarLimit=false
    bUseSphericalLimits=true
    alphaBoolBlend {blendOption=Linear}
    boundBone {boneName=dyn_3_feather_thin_04}
    componentAppliedLinearAccClamp=<2000 2000 2000>
    componentLinearAccScale=<0.3 0.3 0.3>
    componentLinearVelScale=<0.3 0.3 0.3>
    gravityOverride {z=-700}
    physicsBodyDefinitions [{
      boundBone {boneName=dyn_3_feather_thin_04}
      boxExtents=<12 12 12>
      constraintSetup {angularLimitsMax=<10 40 40>, angularLimitsMin=<-15 -40 -40>, angularTarget=<0 0 0>}
      localJointOffset=<2 0 0>
     }]
    sphericalLimits [
     {limitRadius=1, drivingBone {boneName=dyn_feather1_01}, sphereLocalOffset=<3 0 0>}
     {limitRadius=1, drivingBone {boneName=dyn_feather1_02}, sphereLocalOffset=<3 0 0>}
     {limitRadius=1, drivingBone {boneName=dyn_feather2_01}, sphereLocalOffset=<3 0 0>}
     {limitRadius=1, drivingBone {boneName=dyn_feather2_02}, sphereLocalOffset=<3 0 0>}
    ]
   }
   n19 = Anim Dynamics - Bone: dyn_4_feather_thin_01(ComponentPose=@n18, Alpha=1, ExternalForce="0,0,0") {
    angularDampingOverride=0.99
    angularSpringConstant=300
    bAngularSpring=true
    bOverrideAngularDamping=true
    bUseGravityOverride=true
    bUsePlanarLimit=false
    bUseSphericalLimits=true
    alphaBoolBlend {blendOption=Linear}
    boundBone {boneName=dyn_4_feather_thin_01}
    componentAppliedLinearAccClamp=<1000 1000 1000>
    componentLinearAccScale=<0.1 0.1 0.1>
    componentLinearVelScale=<0.1 0.1 0.1>
    gravityOverride {z=-700}
    physicsBodyDefinitions [{
      boundBone {boneName=dyn_4_feather_thin_01}
      boxExtents=<12 12 12>
      constraintSetup {angularLimitsMax=<0 5 5>, angularLimitsMin=<-5 -5 -5>, angularTarget=<0 0 0>}
      localJointOffset=<2 0 0>
     }]
    sphericalLimits [
     {limitRadius=1, drivingBone {boneName=dyn_feather1_01}, sphereLocalOffset=<3 0 0>}
     {limitRadius=1, drivingBone {boneName=dyn_feather1_02}, sphereLocalOffset=<3 0 0>}
     {limitRadius=1, drivingBone {boneName=dyn_feather2_01}, sphereLocalOffset=<3 0 0>}
     {limitRadius=1, drivingBone {boneName=dyn_feather2_02}, sphereLocalOffset=<3 0 0>}
    ]
   }
   n20 = Anim Dynamics - Bone: dyn_4_feather_thin_02(ComponentPose=@n19, Alpha=1, ExternalForce="0,0,0") {
    angularDampingOverride=0.99
    angularSpringConstant=300
    bAngularSpring=true
    bOverrideAngularDamping=true
    bUseGravityOverride=true
    bUsePlanarLimit=false
    bUseSphericalLimits=true
    alphaBoolBlend {blendOption=Linear}
    boundBone {boneName=dyn_4_feather_thin_02}
    componentAppliedLinearAccClamp=<1000 1000 1000>
    componentLinearAccScale=<0.1 0.1 0.1>
    componentLinearVelScale=<0.1 0.1 0.1>
    gravityOverride {z=-700}
    physicsBodyDefinitions [{
      boundBone {boneName=dyn_4_feather_thin_02}
      boxExtents=<12 12 12>
      constraintSetup {angularLimitsMax=<0 20 20>, angularLimitsMin=<-15 -15 15>, angularTarget=<0 0 0>}
      localJointOffset=<2 0 0>
     }]
    sphericalLimits [
     {limitRadius=1, drivingBone {boneName=dyn_feather1_01}, sphereLocalOffset=<3 0 0>}
     {limitRadius=1, drivingBone {boneName=dyn_feather1_02}, sphereLocalOffset=<3 0 0>}
     {limitRadius=1, drivingBone {boneName=dyn_feather2_01}, sphereLocalOffset=<3 0 0>}
     {limitRadius=1, drivingBone {boneName=dyn_feather2_02}, sphereLocalOffset=<3 0 0>}
    ]
   }
   n21 = Anim Dynamics - Bone: dyn_4_feather_thin_03(ComponentPose=@n20, Alpha=1, ExternalForce="0,0,0") {
    angularDampingOverride=0.99
    angularSpringConstant=300
    bAngularSpring=true
    bOverrideAngularDamping=true
    bUseGravityOverride=true
    bUsePlanarLimit=false
    bUseSphericalLimits=true
    alphaBoolBlend {blendOption=Linear}
    boundBone {boneName=dyn_4_feather_thin_03}
    componentAppliedLinearAccClamp=<1500 1500 1500>
    componentLinearAccScale=<0.25 0.25 0.25>
    componentLinearVelScale=<0.25 0.25 25>
    gravityOverride {z=-700}
    physicsBodyDefinitions [{
      boundBone {boneName=dyn_4_feather_thin_03}
      boxExtents=<12 12 12>
      constraintSetup {angularLimitsMax=<10 40 40>, angularLimitsMin=<-15 -40 -40>, angularTarget=<0 0 0>}
      localJointOffset=<2 0 0>
     }]
    sphericalLimits [
     {limitRadius=1, drivingBone {boneName=dyn_feather1_01}, sphereLocalOffset=<3 0 0>}
     {limitRadius=1, drivingBone {boneName=dyn_feather1_02}, sphereLocalOffset=<3 0 0>}
     {limitRadius=1, drivingBone {boneName=dyn_feather2_01}, sphereLocalOffset=<3 0 0>}
     {limitRadius=1, drivingBone {boneName=dyn_feather2_02}, sphereLocalOffset=<3 0 0>}
    ]
   }
   n22 = Anim Dynamics - Bone: dyn_4_feather_thin_04(ComponentPose=@n21, Alpha=1, ExternalForce="0,0,0") {
    angularDampingOverride=0.99
    angularSpringConstant=300
    bAngularSpring=true
    bOverrideAngularDamping=true
    bUseGravityOverride=true
    bUsePlanarLimit=false
    bUseSphericalLimits=true
    alphaBoolBlend {blendOption=Linear}
    boundBone {boneName=dyn_4_feather_thin_04}
    componentAppliedLinearAccClamp=<2000 2000 2000>
    componentLinearAccScale=<0.3 0.3 0.3>
    componentLinearVelScale=<0.3 0.3 0.3>
    gravityOverride {z=-700}
    physicsBodyDefinitions [{
      boundBone {boneName=dyn_4_feather_thin_04}
      boxExtents=<12 12 12>
      constraintSetup {angularLimitsMax=<10 40 40>, angularLimitsMin=<-15 -40 -40>, angularTarget=<0 0 0>}
      localJointOffset=<2 0 0>
     }]
    sphericalLimits [
     {limitRadius=1, drivingBone {boneName=dyn_feather1_01}, sphereLocalOffset=<3 0 0>}
     {limitRadius=1, drivingBone {boneName=dyn_feather1_02}, sphereLocalOffset=<3 0 0>}
     {limitRadius=1, drivingBone {boneName=dyn_feather2_01}, sphereLocalOffset=<3 0 0>}
     {limitRadius=1, drivingBone {boneName=dyn_feather2_02}, sphereLocalOffset=<3 0 0>}
    ]
   }
   n23 = Anim Dynamics - Bone: dyn_5_feather_thin_01(ComponentPose=@n22, Alpha=1, ExternalForce="0,0,0") {
    angularDampingOverride=0.99
    angularSpringConstant=300
    bAngularSpring=true
    bOverrideAngularDamping=true
    bUseGravityOverride=true
    bUsePlanarLimit=false
    bUseSphericalLimits=true
    alphaBoolBlend {blendOption=Linear}
    boundBone {boneName=dyn_5_feather_thin_01}
    componentAppliedLinearAccClamp=<1000 1000 1000>
    componentLinearAccScale=<0.1 0.1 0.1>
    componentLinearVelScale=<0.1 0.1 0.1>
    gravityOverride {z=-700}
    physicsBodyDefinitions [{
      boundBone {boneName=dyn_5_feather_thin_01}
      boxExtents=<12 12 12>
      constraintSetup {angularLimitsMax=<0 5 5>, angularLimitsMin=<-5 -5 -5>, angularTarget=<0 0 0>}
      localJointOffset=<2 0 0>
     }]
    sphericalLimits [
     {limitRadius=1.1, drivingBone {boneName=dyn_feather1_01}, sphereLocalOffset=<2 0 0>}
     {limitRadius=1.1, drivingBone {boneName=dyn_feather1_02}, sphereLocalOffset=<2 0 0>}
     {limitRadius=1.1, drivingBone {boneName=dyn_feather2_01}, sphereLocalOffset=<2 0 0>}
     {limitRadius=1.1, drivingBone {boneName=dyn_feather2_02}, sphereLocalOffset=<2 0 0>}
    ]
   }
   n24 = Anim Dynamics - Bone: dyn_5_feather_thin_02(ComponentPose=@n23, Alpha=1, ExternalForce="0,0,0") {
    angularDampingOverride=0.99
    angularSpringConstant=300
    bAngularSpring=true
    bOverrideAngularDamping=true
    bUseGravityOverride=true
    bUsePlanarLimit=false
    bUseSphericalLimits=true
    alphaBoolBlend {blendOption=Linear}
    boundBone {boneName=dyn_5_feather_thin_02}
    componentAppliedLinearAccClamp=<1000 1000 1000>
    componentLinearAccScale=<0.1 0.1 0.1>
    componentLinearVelScale=<0.1 0.1 0.1>
    gravityOverride {z=-700}
    physicsBodyDefinitions [{
      boundBone {boneName=dyn_5_feather_thin_02}
      boxExtents=<12 12 12>
      constraintSetup {angularLimitsMax=<0 20 20>, angularLimitsMin=<-15 -15 15>, angularTarget=<0 0 0>}
      localJointOffset=<2 0 0>
     }]
    sphericalLimits [
     {limitRadius=1.1, drivingBone {boneName=dyn_feather1_01}, sphereLocalOffset=<2 0 0>}
     {limitRadius=1.1, drivingBone {boneName=dyn_feather1_02}, sphereLocalOffset=<2 0 0>}
     {limitRadius=1.1, drivingBone {boneName=dyn_feather2_01}, sphereLocalOffset=<2 0 0>}
     {limitRadius=1.1, drivingBone {boneName=dyn_feather2_02}, sphereLocalOffset=<2 0 0>}
    ]
   }
   n25 = Anim Dynamics - Bone: dyn_5_feather_thin_03(ComponentPose=@n24, Alpha=1, ExternalForce="0,0,0") {
    angularDampingOverride=0.99
    angularSpringConstant=300
    bAngularSpring=true
    bOverrideAngularDamping=true
    bUseGravityOverride=true
    bUsePlanarLimit=false
    bUseSphericalLimits=true
    alphaBoolBlend {blendOption=Linear}
    boundBone {boneName=dyn_5_feather_thin_03}
    componentAppliedLinearAccClamp=<1500 1500 1500>
    componentLinearAccScale=<0.25 0.25 0.25>
    componentLinearVelScale=<0.25 0.25 25>
    gravityOverride {z=-700}
    physicsBodyDefinitions [{
      boundBone {boneName=dyn_5_feather_thin_03}
      boxExtents=<12 12 12>
      constraintSetup {angularLimitsMax=<10 40 40>, angularLimitsMin=<-15 -40 -40>, angularTarget=<0 0 0>}
      localJointOffset=<2 0 0>
     }]
    sphericalLimits [
     {limitRadius=1.1, drivingBone {boneName=dyn_feather1_01}, sphereLocalOffset=<2 0 0>}
     {limitRadius=1.1, drivingBone {boneName=dyn_feather1_02}, sphereLocalOffset=<2 0 0>}
     {limitRadius=1.1, drivingBone {boneName=dyn_feather2_01}, sphereLocalOffset=<2 0 0>}
     {limitRadius=1.1, drivingBone {boneName=dyn_feather2_02}, sphereLocalOffset=<2 0 0>}
    ]
   }
   n26 = Anim Dynamics - Bone: dyn_5_feather_thin_04(ComponentPose=@n25, Alpha=1, ExternalForce="0,0,0") {
    angularDampingOverride=0.99
    angularSpringConstant=300
    bAngularSpring=true
    bOverrideAngularDamping=true
    bUseGravityOverride=true
    bUsePlanarLimit=false
    bUseSphericalLimits=true
    alphaBoolBlend {blendOption=Linear}
    boundBone {boneName=dyn_5_feather_thin_04}
    componentAppliedLinearAccClamp=<2000 2000 2000>
    componentLinearAccScale=<0.3 0.3 0.3>
    componentLinearVelScale=<0.3 0.3 0.3>
    gravityOverride {z=-700}
    physicsBodyDefinitions [{
      boundBone {boneName=dyn_5_feather_thin_04}
      boxExtents=<12 12 12>
      constraintSetup {angularLimitsMax=<10 40 40>, angularLimitsMin=<-15 -40 -40>, angularTarget=<0 0 0>}
      localJointOffset=<2 0 0>
     }]
    sphericalLimits [
     {limitRadius=1.1, drivingBone {boneName=dyn_feather1_01}, sphereLocalOffset=<2 0 0>}
     {limitRadius=1.1, drivingBone {boneName=dyn_feather1_02}, sphereLocalOffset=<2 0 0>}
     {limitRadius=1.1, drivingBone {boneName=dyn_feather2_01}, sphereLocalOffset=<2 0 0>}
     {limitRadius=1.1, drivingBone {boneName=dyn_feather2_02}, sphereLocalOffset=<2 0 0>}
    ]
   }
   n27 = Anim Dynamics - Bone: dyn_6_feather_thin_01(ComponentPose=@n26, Alpha=1, ExternalForce="0,0,0") {
    angularDampingOverride=0.99
    angularSpringConstant=300
    bAngularSpring=true
    bOverrideAngularDamping=true
    bUseGravityOverride=true
    bUsePlanarLimit=false
    bUseSphericalLimits=true
    alphaBoolBlend {blendOption=Linear}
    boundBone {boneName=dyn_6_feather_thin_01}
    componentAppliedLinearAccClamp=<1000 1000 1000>
    componentLinearAccScale=<0.1 0.1 0.1>
    componentLinearVelScale=<0.1 0.1 0.1>
    gravityOverride {z=-700}
    physicsBodyDefinitions [{
      boundBone {boneName=dyn_6_feather_thin_01}
      boxExtents=<12 12 12>
      constraintSetup {angularLimitsMax=<0 5 5>, angularLimitsMin=<-5 -5 -5>, angularTarget=<0 0 0>}
      localJointOffset=<2 0 0>
     }]
    sphericalLimits [
     {limitRadius=1, drivingBone {boneName=dyn_feather1_01}, sphereLocalOffset=<3 0 0>}
     {limitRadius=1, drivingBone {boneName=dyn_feather1_02}, sphereLocalOffset=<3 0 0>}
     {limitRadius=1, drivingBone {boneName=dyn_feather2_01}, sphereLocalOffset=<3 0 0>}
     {limitRadius=1, drivingBone {boneName=dyn_feather2_02}, sphereLocalOffset=<3 0 0>}
    ]
   }
   n28 = Anim Dynamics - Bone: dyn_6_feather_thin_02(ComponentPose=@n27, Alpha=1, ExternalForce="0,0,0") {
    angularDampingOverride=0.99
    angularSpringConstant=300
    bAngularSpring=true
    bOverrideAngularDamping=true
    bUseGravityOverride=true
    bUsePlanarLimit=false
    bUseSphericalLimits=true
    alphaBoolBlend {blendOption=Linear}
    boundBone {boneName=dyn_6_feather_thin_02}
    componentAppliedLinearAccClamp=<1000 1000 1000>
    componentLinearAccScale=<0.1 0.1 0.1>
    componentLinearVelScale=<0.1 0.1 0.1>
    gravityOverride {z=-700}
    physicsBodyDefinitions [{
      boundBone {boneName=dyn_6_feather_thin_02}
      boxExtents=<12 12 12>
      constraintSetup {angularLimitsMax=<0 20 20>, angularLimitsMin=<-15 -5 -5>, angularTarget=<0 0 0>}
      localJointOffset=<2 0 0>
     }]
    sphericalLimits [
     {limitRadius=1, drivingBone {boneName=dyn_feather1_01}, sphereLocalOffset=<3 0 0>}
     {limitRadius=1, drivingBone {boneName=dyn_feather1_02}, sphereLocalOffset=<3 0 0>}
     {limitRadius=1, drivingBone {boneName=dyn_feather2_01}, sphereLocalOffset=<3 0 0>}
     {limitRadius=1, drivingBone {boneName=dyn_feather2_02}, sphereLocalOffset=<3 0 0>}
    ]
   }
   n29 = Anim Dynamics - Bone: dyn_6_feather_thin_03(ComponentPose=@n28, Alpha=1, ExternalForce="0,0,0") {
    angularDampingOverride=0.99
    angularSpringConstant=300
    bAngularSpring=true
    bOverrideAngularDamping=true
    bUseGravityOverride=true
    bUsePlanarLimit=false
    bUseSphericalLimits=true
    alphaBoolBlend {blendOption=Linear}
    boundBone {boneName=dyn_6_feather_thin_03}
    componentAppliedLinearAccClamp=<1000 1000 1000>
    componentLinearAccScale=<0.1 0.1 0.1>
    componentLinearVelScale=<0.1 0.1 0.1>
    gravityOverride {z=-700}
    physicsBodyDefinitions [{
      boundBone {boneName=dyn_6_feather_thin_03}
      boxExtents=<12 12 12>
      constraintSetup {angularLimitsMax=<0 20 20>, angularLimitsMin=<-15 -15 15>, angularTarget=<0 0 0>}
      localJointOffset=<2 0 0>
     }]
    sphericalLimits [
     {limitRadius=1, drivingBone {boneName=dyn_feather1_01}, sphereLocalOffset=<3 0 0>}
     {limitRadius=1, drivingBone {boneName=dyn_feather1_02}, sphereLocalOffset=<3 0 0>}
     {limitRadius=1, drivingBone {boneName=dyn_feather2_01}, sphereLocalOffset=<3 0 0>}
     {limitRadius=1, drivingBone {boneName=dyn_feather2_02}, sphereLocalOffset=<3 0 0>}
    ]
   }
   n30 = Anim Dynamics - Bone: dyn_6_feather_thin_04(ComponentPose=@n29, Alpha=1, ExternalForce="0,0,0") {
    angularDampingOverride=0.99
    angularSpringConstant=300
    bAngularSpring=true
    bOverrideAngularDamping=true
    bUseGravityOverride=true
    bUsePlanarLimit=false
    bUseSphericalLimits=true
    alphaBoolBlend {blendOption=Linear}
    boundBone {boneName=dyn_6_feather_thin_04}
    componentAppliedLinearAccClamp=<1500 1500 1500>
    componentLinearAccScale=<0.25 0.25 0.25>
    componentLinearVelScale=<0.25 0.25 25>
    gravityOverride {z=-700}
    physicsBodyDefinitions [{
      boundBone {boneName=dyn_6_feather_thin_04}
      boxExtents=<12 12 12>
      constraintSetup {angularLimitsMax=<10 40 40>, angularLimitsMin=<-15 -40 -40>, angularTarget=<0 0 0>}
      localJointOffset=<2 0 0>
     }]
    sphericalLimits [
     {limitRadius=1, drivingBone {boneName=dyn_feather1_01}, sphereLocalOffset=<3 0 0>}
     {limitRadius=1, drivingBone {boneName=dyn_feather1_02}, sphereLocalOffset=<3 0 0>}
     {limitRadius=1, drivingBone {boneName=dyn_feather2_01}, sphereLocalOffset=<3 0 0>}
     {limitRadius=1, drivingBone {boneName=dyn_feather2_02}, sphereLocalOffset=<3 0 0>}
    ]
   }
   n31 = Anim Dynamics - Bone: dyn_6_feather_thin_05(ComponentPose=@n30, Alpha=1, ExternalForce="0,0,0") {
    angularDampingOverride=0.99
    angularSpringConstant=300
    bAngularSpring=true
    bOverrideAngularDamping=true
    bUseGravityOverride=true
    bUsePlanarLimit=false
    bUseSphericalLimits=true
    alphaBoolBlend {blendOption=Linear}
    boundBone {boneName=dyn_6_feather_thin_05}
    componentAppliedLinearAccClamp=<2000 2000 2000>
    componentLinearAccScale=<0.3 0.3 0.3>
    componentLinearVelScale=<0.3 0.3 0.3>
    gravityOverride {z=-700}
    physicsBodyDefinitions [{
      boundBone {boneName=dyn_6_feather_thin_05}
      boxExtents=<12 12 12>
      constraintSetup {angularLimitsMax=<10 40 40>, angularLimitsMin=<-15 -40 -40>, angularTarget=<0 0 0>}
      localJointOffset=<2 0 0>
     }]
    sphericalLimits [
     {limitRadius=1, drivingBone {boneName=dyn_feather1_01}, sphereLocalOffset=<3 0 0>}
     {limitRadius=1, drivingBone {boneName=dyn_feather1_02}, sphereLocalOffset=<3 0 0>}
     {limitRadius=1, drivingBone {boneName=dyn_feather2_01}, sphereLocalOffset=<3 0 0>}
     {limitRadius=1, drivingBone {boneName=dyn_feather2_02}, sphereLocalOffset=<3 0 0>}
    ]
   }
   n32 = RigidBody(ComponentPose=Local To Component(LocalPose=Use cached pose 'Base'())) {
    bClampLinearTranslationLimitToRefPose=true
    bOverrideWorldGravity=true
    overridePhysicsAsset=&/FineCheek_Axe_Physics
    alphaBoolBlend {blendOption=Linear}
    componentLinearAccScale=<1 1 1>
    componentLinearVelScale=<1 1 1>
    overrideWorldGravity {z=-900}
   }
   Output Pose(Result=Blend Poses by bool(BlendPose_0=Component To Local(ComponentPose=@n32), BlendPose_1=Component To Local(ComponentPose=@n31), bActiveValue=Get bIsInFrontEnd()) {blendType=Linear}) {layerGroup=None}
   AnimGraphNode_SubInput_0 = Input Pose()
   Save cached pose 'Base'(Pose=@AnimGraphNode_SubInput_0) {cacheName=Base}
  ]}
 EventGraph {nodes [
   Event BlueprintUpdateAnimation()
   > Set Jumping(Jumping:bool=OR Boolean(A=Get bIsJumping(), B=Get bIsFalling())) -> (Output_Get:bool)
   TryGetPawnOwner()
  ]}
]
variables [Jumping {type=boolean}]
